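# Physics asset for the RecordScratch backpack cosmetic, defining capsule bodies and joints for its dangling glasses and zipper pulls.
# A large pack capsule is the root for three small base bodies, joined with free swing and twist.
# Each base carries a dynamic-named bone with limited swing and very tight twist.
# Zipper A continues into a four-link ball chain, while glasses and zipper B are single links.
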PhysicsAsset /BRCosmetics/Accessories/FORT_Backpacks/Backpack_F_MED_RecordScratch/Meshes/F_MED_RecordScratch_BackPack_Physics
bodies {
 base_glasses [{shapeType=Capsule, length=0.505, radius=0.505, shapeName=base_glasses_capsule, rotation=<rot 0 -90 0>}]
 base_zipper_a [{shapeType=Capsule, length=0.505, radius=0.505, shapeName=base_zipper_a_capsule, rotation=<rot 0 -90 0>}]
 base_zipper_b [{shapeType=Capsule, length=0.505, radius=0.505, shapeName=base_zipper_b_capsule, rotation=<rot 0 -90 0>}]
 dyn_ball_00 [{shapeType=Capsule, length=2.3, radius=0.7, shapeName=dyn_ball_00_capsule, center=<2 0 0>, rotation=<rot 90 0 0>}]
 dyn_ball_01 [{shapeType=Capsule, length=2.4, radius=0.6, shapeName=dyn_ball_01_capsule, center=<2 0 0>, rotation=<rot 90 0 0>}]
 dyn_ball_02 [{shapeType=Capsule, length=2, radius=1, shapeName=dyn_ball_02_capsule, center=<2 0 0>, rotation=<rot 90 0 0>}]
 dyn_ball_03 [{shapeType=Capsule, length=0.4303, radius=2.8, shapeName=dyn_ball_03_capsule, center=<2.8054 0 -0.087>, rotation=<rot 90 0 0>}]
 dyn_glasses [{shapeType=Capsule, length=2, radius=1, shapeName=dyn_glasses_capsule, center=<2 0 0>, rotation=<rot 90 0 0>}]
 dyn_zipper_a [{shapeType=Capsule, length=2.1, radius=0.9, shapeName=dyn_zipper_a_capsule, center=<1.5625 0 0>, rotation=<rot 90 0 0>}]
 dyn_zipper_b [{shapeType=Capsule, length=2.3, radius=0.7, shapeName=dyn_zipper_b_capsule, center=<1.7977 0 0>, rotation=<rot 90 -179.9961 -179.9968>}]
 pack [{shapeType=Capsule, length=0.2704, radius=10.7582, shapeName=pack_capsule, center=<0.6189 -5.4735 -12.9107>, rotation=<rot 0 -90 0>}]
}
constraints [
 {
  bone1Name=base_glasses
  bone2Name=pack
  swing1LimitDegrees=45
  swing1Motion=Free
  swing2LimitDegrees=45
  swing2Motion=Free
  twistLimitDegrees=45
  twistMotion=Free
 }
 {bone1Name=dyn_glasses, bone2Name=base_glasses, swing1LimitDegrees=0.1, swing2LimitDegrees=19, twistLimitDegrees=1}
 {
  bone1Name=base_zipper_a
  bone2Name=pack
  swing1LimitDegrees=45
  swing1Motion=Free
  swing2LimitDegrees=45
  swing2Motion=Free
  twistLimitDegrees=45
  twistMotion=Free
 }
 {bone1Name=dyn_zipper_a, bone2Name=base_zipper_a, swing1LimitDegrees=10, swing2LimitDegrees=30, twistLimitDegrees=1}
 {bone1Name=dyn_ball_00, bone2Name=dyn_zipper_a, swing1LimitDegrees=25, swing2LimitDegrees=30, twistLimitDegrees=1}
 {bone1Name=dyn_ball_01, bone2Name=dyn_ball_00, swing1LimitDegrees=25, swing2LimitDegrees=30, twistLimitDegrees=1}
 {bone1Name=dyn_ball_02, bone2Name=dyn_ball_01, swing1LimitDegrees=25, swing2LimitDegrees=30, twistLimitDegrees=1}
 {bone1Name=dyn_ball_03, bone2Name=dyn_ball_02, swing1LimitDegrees=25, swing2LimitDegrees=30, twistLimitDegrees=1}
 {
  bone1Name=base_zipper_b
  bone2Name=pack
  swing1LimitDegrees=45
  swing1Motion=Free
  swing2LimitDegrees=45
  swing2Motion=Free
  twistLimitDegrees=45
  twistMotion=Free
 }
 {bone1Name=dyn_zipper_b, bone2Name=base_zipper_b, swing1LimitDegrees=10, swing2LimitDegrees=21.544, twistLimitDegrees=1}
]
