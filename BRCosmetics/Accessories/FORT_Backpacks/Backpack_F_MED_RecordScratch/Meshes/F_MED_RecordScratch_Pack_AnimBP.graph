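# Anim Blueprint for the RecordScratch backpack, animating its base_glasses bone and simulating secondary motion.
# With rigid body enabled, an additive pitch scaled by Jump Alpha feeds a RigidBody node using an override physics asset and damped component motion; otherwise a slight fixed pitch applies.
# Each update sets simulation gravity through a deprecated movement-state function and derives Jump Alpha from jumping or falling.
# Directional movement flags are present but unused by the AnimGraph.
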AnimBlueprint /BRCosmetics/Accessories/FORT_Backpacks/Backpack_F_MED_RecordScratch/Meshes/F_MED_RecordScratch_Pack_AnimBP
target_skeleton=&/F_MED_RecordScratch_Pack_Skeleton
graphs [
 AnimGraph {nodes [
   n1 = "Transform (Modify) Bone - Bone: base_glasses"(ComponentPose=Local To Component(LocalPose=Use cached pose 'Input Cache'()), Alpha=1, Translation="0,0,0", Rotation="3,0,0", Scale="1,1,1") {
    rotationMode=BMM_Additive
    rotationSpace=BCS_BoneSpace
    alphaBoolBlend {blendOption=Linear}
    boneToModify {boneName=base_glasses}
    rotation {pitch=3}
   }
   n2 = "Transform (Modify) Bone - Bone: base_glasses"(
    ComponentPose=Local To Component(LocalPose=Use cached pose 'Input Cache'())
    Alpha=Get Jump Alpha() -> (Jump Alpha:double)
    Translation="0,0,0"
    Rotation="25,0,0"
    Scale="1,1,1"
   ) {
    rotationMode=BMM_Additive
    rotationSpace=BCS_BoneSpace
    translationMode=BMM_Additive
    alphaBoolBlend {blendOption=Linear}
    alphaScaleBiasClamp {bInterpResult=true}
    boneToModify {boneName=base_glasses}
    rotation {pitch=25}
   }
   n3 = RigidBody!(ComponentPose=@n2, Alpha, OverrideWorldGravity:Vector=Get GravityOverride(), ExternalForce) -> (Pose) {
    bClampLinearTranslationLimitToRefPose=true
    bOverrideWorldGravity=true
    overridePhysicsAsset=&/F_MED_RecordScratch_BackPack_Physics
    simulationSpace=BaseBoneSpace
    alphaBoolBlend {blendOption=Linear}
    baseBoneRef {boneName=root}
    componentAppliedLinearAccClamp=<3000 3000 3000>
    componentLinearAccScale=<0.3 0.15 0.2>
    componentLinearVelScale=<0.3 0.15 0.2>
    simSpaceSettings {
     maxAngularAcceleration=1000
     maxAngularVelocity=3
     maxLinearAcceleration=700
     maxLinearVelocity=500
     velocityScaleZ=0
     worldAlpha=0.25
     externalLinearDragV=<0.2 0.2>
    }
   }
   Output Pose(Result=Blend Poses by bool(BlendPose_0=Component To Local(ComponentPose=@n3), BlendPose_1=Component To Local(ComponentPose=@n1), bActiveValue=Get bRigidBodyEnabled()) {blendType=Linear}) {layerGroup=None}
   Save cached pose 'Input Cache'(Pose=Input Pose()) {cacheName=Input Cache}
  ]}
 EventGraph {nodes [
   Event BlueprintUpdateAnimation()
   > n1 = Sequence!(execute) -> (then_0, then_1, then_2:Exec)
   > n2 = Gravity Override State Function DEPRECATED(CharacterPartAnimInstance=Self-Reference(), GameTimeInSeconds=GetGameTimeInSeconds(), GravityOverrideParams=Get Gravity Override Params())
   > Set GravityOverride(GravityOverride=@n2)
   n3 = PlayerMovement(execute=@n1.then_1, AnimInstance=Self-Reference(), XDotProductCoefficient=1, YDotProductCoefficient=1, ZDotProductCoefficient=1)
   n4 = "To Float (Boolean)"(InBool=OR Boolean(A=Get bIsJumping(), B=Get bIsFalling()))
   Set Jump Alpha(execute=@n1.then_2, Jump Alpha:double=@n4) -> (Output_Get:double)
   TryGetPawnOwner()
   float * float(
    A=@n4
    B=float + float!(A=Get Left() -> (Left:double), B=Get Right() -> (Right:double), C:double=Get Backward() -> (Backward:double)) -> (ReturnValue)
   ) {is=title K2Node_PromotableOperator}
   n5 = BreakVector(InVec=MakeVector(X=@n3, Y=@n3.YMovement))
   Set Left(Left:double="To Float (Boolean)"(InBool="float < float"(A=@n5, B=-50))) -> (Output_Get:double)
   Set Right(Right:double="To Float (Boolean)"(InBool="float > float"(A=@n5, B=50))) -> (Output_Get:double)
   Set Forward(Forward:double="To Float (Boolean)"(InBool="float > float"(A=@n5.Y, B=40))) -> (Output_Get:double)
   Set Backward(Backward:double="To Float (Boolean)"(InBool="float < float"(A=@n5.Y, B=-40))) -> (Output_Get:double)
  ]}
]
variables [
 Backward {type=number}
 Forward {type=number}
 Gravity Override Params {
  type=GravityOverrideParamsStruct
  default {onGround_Falling {gravityOverride=<0 0 -350>}, onGround_Moving {gravityOverride=<0 0 -800>}, onGround_Standing {gravityOverride=<0 0 -450>}}
 }
 GravityOverride {type=Vector}
 Jump Alpha {type=number}
 Left {type=number}
 Right {type=number}
]
